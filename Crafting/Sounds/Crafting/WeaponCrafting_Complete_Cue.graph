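# Sound cue for completing weapon crafting, using consumable-use attenuation and a loot reward sound class.
# A Mixer layers the completion wave with the crafting-open wave, weighting the open layer slightly louder.
# Only the open wave passes through a Modulator, which randomly raises its pitch on each play so repeated crafts vary.
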SoundCue /Crafting/Sounds/Crafting/WeaponCrafting_Complete_Cue
root=Mixer_1
nodes [
 {id=Mixer_1, children [WavePlayer_8,Modulator_0], properties {inputVolume [1,1.2]}}
 {id=Modulator_0, children [WavePlayer_3], properties {pitchMax=1.3, pitchMin=1.1}}
 {id=WavePlayer_3, properties {soundWaveAssetPtr=&/WeaponCrafting_Open_01}}
 {id=WavePlayer_8, properties {soundWaveAssetPtr=&/WeaponCrafting_Complete_02}}
]
properties {
 attenuationSettings=&Fort_Consumable_Use_Athena_Att
 maxDistance=2250
 soundClassObject=&Fort_Loot_Reward_PSM
 volumeMultiplier=0.85
 attenuationOverrides {falloffDistance=1200, attenuationShapeExtents {x=800}}
 concurrencyOverrides {resolutionRule=PreventNew}
}
