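# Sound cue for the weapon crafting foley sound, played as a player foley effect.
# A single wave player feeds the WeaponCraft_Foley_01 wave straight to the output with no randomization or mixing.
# It uses the Athena consumable-use attenuation, the player foley sound class and a boosted volume.
# Concurrency overrides set a small instance cap that stops the oldest sound.
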SoundCue /Crafting/Sounds/Crafting/WeaponCrafting_Foley_Cue
root=WavePlayer_0
nodes [{id=WavePlayer_0, properties {soundWaveAssetPtr=&/WeaponCraft_Foley_01}}]
properties {
 attenuationSettings=&Fort_Consumable_Use_Athena_Att
 maxDistance=2250
 soundClassObject=&Fort_Player_Foley
 volumeMultiplier=2.9867
 concurrencyOverrides {maxCount=3, resolutionRule=StopOldest}
}
