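# Sound cue for the weapon crafting hit, played as player foley with attenuation borrowed from consumable use.
# A quality-level node routes the first two levels to an equally weighted random pick of five hit waves, and the last two to a pick of three.
# The three-wave random node is excluded from branch culling.
# Volume is boosted, and a concurrency override caps simultaneous plays, stopping the oldest.
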SoundCue /Crafting/Sounds/Crafting/WeaponCrafting_Hit_Cue
root=QualityLevel_1
nodes [
 {id=QualityLevel_1, children [Random_1,Random_1,Random_2,Random_2]}
 {id=Random_1, children [WavePlayer_2,WavePlayer_3,WavePlayer_4,WavePlayer_5,WavePlayer_6], properties {weights [1,1,1,1,1]}}
 {id=Random_2, children [WavePlayer_8, WavePlayer_10, WavePlayer_11], properties {bShouldExcludeFromBranchCulling=true, weights [1,1,1]}}
 {id=WavePlayer_10, properties {soundWaveAssetPtr=&/WeaponCrafting_Hit_03}}
 {id=WavePlayer_11, properties {soundWaveAssetPtr=&/WeaponCrafting_Hit_04}}
 {id=WavePlayer_2, properties {soundWaveAssetPtr=&/WeaponCrafting_Hit_05}}
 {id=WavePlayer_3, properties {soundWaveAssetPtr=&/WeaponCrafting_Hit_01}}
 {id=WavePlayer_4, properties {soundWaveAssetPtr=&/WeaponCrafting_Hit_02}}
 {id=WavePlayer_5, properties {soundWaveAssetPtr=&WeaponCrafting_Hit_03}}
 {id=WavePlayer_6, properties {soundWaveAssetPtr=&WeaponCrafting_Hit_04}}
 {id=WavePlayer_8, properties {soundWaveAssetPtr=&WeaponCrafting_Hit_01}}
]
properties {
 attenuationSettings=&Fort_Consumable_Use_Athena_Att
 maxDistance=2250
 soundClassObject=&Fort_Player_Foley
 volumeMultiplier=2.9867
 concurrencyOverrides {maxCount=3, resolutionRule=StopOldest}
}
